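# Niagara system for the Rage Unleashed vehicle boost cosmetic, trailing effects from the car's boost sockets.
# CoreRibbons and LightningRibbons spawn per port and stream backward, following vehicle motion, with age-driven width, color and sine-wave position offsets for multiple ribbon renderers.
# A third emitter spawns mesh and sprite particles at the sockets, spinning in opposite directions per port.
# All emitters apply painted-variant color overrides; user parameters supply car offset, preview state and lightning wave shape.
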NiagaraSystem /VehicleCosmetics/Boosts/Boost_RageUnleashed/FX/NS_Boost_RageUnleashed1
dependencies {
 usedDataInterfaces [
  &NiagaraDataInterfaceArrayColor
  &NiagaraDataInterfaceArrayFloat
  &NiagaraDataInterfaceCurve
  &NiagaraDataInterfaceColorCurve
  &NiagaraDataInterfaceSkeletalMesh
 ]
 usedDynamicInputs []
 usedModules [
  &self:SystemSpawnScript.SystemScriptSource.SystemScriptGraph.NiagaraNodeAssignment_0.SetVariables_#1
  &SystemState
 ]
}
emitters {
 CoreRibbons {
  emitterSpawnScript [/SetVariables(Emitter.CoreWidth=40) {isSetParameters=true, module=SetVariables_#2}]
  emitterUpdateScript [
   &EmitterState()
   &VC_FilteredSocketCount()
   &VC_RateQualityHandler(High Spawn Rate=60, Low Spawn Rate=60, Medium Spawn Rate=60, Use Custom Rate?=true)
   &VC_SpawnRibbon_v1(Number of Ribbons=$Emitter.PortCount, SpawnRate=$Emitter.Spawn Rate)
  ]
  particleSpawnScript [
   &VC_PaintedVariantsOverride(
    Black=<rgba 0.1 0.1 0.1 1.5>
    BurntSienna=<rgba 0.197 0.0877 0.0369 1>
    Cobalt=<rgba 0.1983 0.211 0.8854 1>
    Crimson=<rgba 0.4635 0.0604 0.0713 2>
    DefaultColor=<rgba 0.7 0.1021 0.1021 1>
    ForestGreen=<rgba 0.0101 0.4844 0.0341 0.9>
    Lime=<rgba 0.3148 0.5 0 1.5>
    Orange=<rgba 0.6724 0.2346 0 1>
    Pink=<rgba 0.7396 0.0886 0.5029 1>
    Purple=<rgba 0.2629 0 0.6406 1.3>
    R1 Black:LinearColor=<rgba 0.1 0.1 0.1 1>
    R1 BurntSienna:LinearColor=<rgba 0.7 0.334 0.2499 1>
    R1 Cobalt:LinearColor=<rgba 0 0.037 2 5>
    R1 Crimson:LinearColor=<rgba 2 0.4583 0.5001 1>
    R1 DefaultColor=<rgba 1.4 0.56 0.35 1>
    R1 ForestGreen:LinearColor=<rgba 0.25 2 0.3384 1>
    R1 Grey:LinearColor=<rgba 0.3646 0.3646 0.3646 1>
    R1 Lime:LinearColor=<rgba 0.6297 0.8 0 2>
    R1 Orange:LinearColor=<rgba 1 0.4681 0 2>
    R1 Pink:LinearColor=<rgba 1 0.4115 0.7563 2>
    R1 Purple:LinearColor=<rgba 0.7618 0.2312 1.2 1>
    R1 Saffron:LinearColor=<rgba 1 0.859 0.3958 1.2>
    R1 SkyBlue:LinearColor=<rgba 0.3094 0.7024 0.9 1.5>
    Saffron=<rgba 0.526 0.4615 0.2493 2>
    SkyBlue=<rgba 0 0.6824 0.901 0.5>
    TitaniumWhite=<rgba 0.3906 0.3906 0.3906 1.2>
    Use R1 Painted Overrides=true
   )
   &VC_InitializeRibbon_v1(
    Color=$Particles.Color
    Custom Offset=true
    Enable Skeletal Position=true
    Lifetime=0.6
    Local Vector Offset=&AddVectorToPosition(
     Position=<0 0 0>
     Vector=&AddVectorToPosition(
      Position=$User.DTCarOffset
      Vector=&MakeCustomVectorFromBool(Bool=$User.bIsFrontendPreview, True Vector=<18 0 0>, False Vector=<0 0 0>)
     )
    )
    Offset Amount=1
    Ribbon Width=75
   )
   &AddVelocity(Velocity=<-2000 0 0>) {module=AddVelocity001}
  ]
  particleUpdateScript [
   &ParticleState()
   &InheritSourceMovement(
    Applied Position Scale=&Multiply_VectorByFloat(
     Vector=<1 1 1>
     Float=&FloatFromCurve(FloatCurve=DataInterface() {propertyValues {Curve {keys [
          {interpMode=RCIM_Cubic, leaveTangent=-0.0051, tangentMode=RCTM_Break, time=0, value=1}
          {arriveTangent=-0.0476, interpMode=RCIM_Cubic, tangentMode=RCTM_Break, time=1, value=0.9}
         ]}}})
    )
    ApplyForce=false
    ApplyPosition=true
    ApplyPositionFalloff=false
   )
   &ScaleColor(
    Scale Alpha=&FloatFromCurve(FloatCurve=DataInterface() {propertyValues {Curve {keys [
         {arriveTangent=-0.0712, interpMode=RCIM_Cubic, leaveTangent=-0.0712, tangentMode=RCTM_User, time=0.05, value=0}
         {arriveTangent=0.0402, interpMode=RCIM_Cubic, leaveTangent=0.0402, tangentMode=RCTM_User, time=0.2, value=1}
         {arriveTangent=-0.25, interpMode=RCIM_Cubic, leaveTangent=-0.25, time=0.5, value=1}
         {arriveTangent=-2, interpMode=RCIM_Cubic, leaveTangent=-2, time=0.6, value=0.9}
         {interpMode=RCIM_Cubic, leaveTangent=-6.4238, time=1, value=0}
        ]}}})
    ScaleA=true
    ScaleRGB=true
   )
   &SolveForcesAndVelocity(Clamp Velocity=false, Limit Acceleration=false)
   &ScaleRibbonWidth(Ribbon Width Scale=&FloatFromCurve(FloatCurve=DataInterface() {propertyValues {Curve {keys [
         <at 0.05 0 RCIM_Cubic>
         {arriveTangent=4.2857, interpMode=RCIM_Cubic, leaveTangent=4.2857, time=0.15, value=1.8}
         {arriveTangent=-1.4118, interpMode=RCIM_Cubic, leaveTangent=-1.4118, time=0.4, value=1.5}
         <at 1 0.6 RCIM_Cubic>
        ]}}}))
   &VC_UpdateRibbon_v1(Enable Skeletal Position=true)
   /SetVariables(
    Particles.PositionR1:NiagaraPosition=&AddVectorToPosition(
     Position=$Particles.Position
     Vector=&TransformVector(
      Vector=&MakeVector(
       X=0
       Y=&Sine(
        Normalized Angle=&Add_Float(
         A=&Multiply_Float(A=$Particles.NormalizedAge, B=-1.3)
         B=&Multiply_Float(A=$Emitter.LoopedAge, B=1)
        )
        Period=0.5
        Scale=&FloatFromCurve(
         FloatCurve=DataInterface() {propertyValues {Curve {keys [<at 0.1 0>,<at 0.6 2>,<at 1 5>]}}}
         Scale Curve=&MaskFloatBySpawnGroup(Masked Float=6, Passthrough Float=-6)
        )
        Bias=0
       )
       Z=&Sine(
        Normalized Angle=&Add_Float(
         A=&Multiply_Float(A=$Particles.NormalizedAge, B=-1.3)
         B=&Multiply_Float(A=$Emitter.LoopedAge, B=1)
        )
        Period=0.5
        Scale=&FloatFromCurve(
         FloatCurve=DataInterface() {propertyValues {Curve {keys [<at 0.1 0>,<at 0.5 2>,<at 1 2.5>]}}}
         Scale Curve=&MaskFloatBySpawnGroup(Masked Float=5, Passthrough Float=-5)
        )
        Bias=0
       )
      )
      Destination Space=::Simulation
      Apply Scale=true
     )
    )
    Particles.PositionR2:NiagaraPosition=&AddVectorToPosition(
     Position=$Particles.Position
     Vector=&TransformVector(
      Vector=&MakeVector(
       X=-10
       Y=&Sine(
        Normalized Angle=&Add_Float(
         A=&Multiply_Float(A=$Particles.NormalizedAge, B=-1.3)
         B=&Multiply_Float(A=$Emitter.LoopedAge, B=1)
        )
        Period=0.3
        Scale=&FloatFromCurve(
         FloatCurve=DataInterface() {propertyValues {Curve {keys [<at 0.1 0>,<at 0.5 1>,<at 1 3>]}}}
         Scale Curve=&MaskFloatBySpawnGroup(Masked Float=-8, Passthrough Float=8)
        )
        Bias=0
       )
       Z=&Sine(
        Normalized Angle=&Add_Float(
         A=&Multiply_Float(A=$Particles.NormalizedAge, B=-1.3)
         B=&Multiply_Float(A=$Emitter.LoopedAge, B=1)
        )
        Period=0.3
        Scale=&FloatFromCurve(
         FloatCurve=DataInterface() {propertyValues {Curve {keys [<at 0.1 0>,<at 0.5 1>]}}}
         Scale Curve=&MaskFloatBySpawnGroup(Masked Float=-8, Passthrough Float=8)
        )
        Bias=0
       )
      )
      Destination Space=::Simulation
      Apply Scale=true
     )
    )
   ) {isSetParameters=true}
   /SetVariables(
    Particles.CoreColor:LinearColor=&MultiplyLinearColor(
     Linear Color=$Particles.R1 Color
     Scale Factor=&Vector4FromCurve(Vector4Curve=DataInterface() {propertyValues {
        AlphaCurve {keys [<at 0.1518 0>,<at 0.215 0.5>,<at 0.6088 0>]}
        BlueCurve {keys [<at 0 1>,<at 1 1>]}
        GreenCurve {keys [<at 0 1>,<at 1 1>]}
        RedCurve {keys [<at 0 1>,<at 1 1>]}
       }})
    )
    Particles.CoreRibbonSize:NiagaraFloat=&Multiply_Float(
     A=$Emitter.CoreWidth
     B=&FloatFromCurve(
      FloatCurve=DataInterface() {propertyValues {Curve {keys [
          {arriveTangent=-2.375, interpMode=RCIM_Cubic, time=0.05, value=0}
          {arriveTangent=-0.0286, interpMode=RCIM_Cubic, leaveTangent=-0.0286, tangentMode=RCTM_User, time=0.2, value=2.5}
          <at 1 0.6>
         ]}}}
      Scale Curve=0.8
     )
    )
   ) {isSetParameters=true, module=SetVariables_#3}
  ]
  rendererClasses [&NiagaraRibbonRendererProperties]
  renderers [
   {rendererClass=&NiagaraRibbonRendererProperties}
   {rendererClass=&NiagaraRibbonRendererProperties, rendererIndex=1}
  ]
 }
 LightningRibbons {
  emitterSpawnScript [/SetVariables(Emitter.CoreWidth=70) {isSetParameters=true, module=SetVariables_#2}]
  emitterUpdateScript [
   &EmitterState()
   &VC_FilteredSocketCount()
   &VC_RateQualityHandler(High Spawn Rate=60, Low Spawn Rate=60, Medium Spawn Rate=60, Use Custom Rate?=true)
   &VC_SpawnRibbon_v1(Number of Ribbons=$Emitter.PortCount, Spawn Time=-1, SpawnRate=$Emitter.Spawn Rate)
  ]
  particleSpawnScript [
   &VC_PaintedVariantsOverride(
    Black=<rgba 0.0191 0.0191 0.0191 2>
    BurntSienna=<rgba 0.8087 0.3295 0.1547 1>
    Cobalt=<rgba 0 0.0555 3 1>
    Crimson=<rgba 1.8 0.0375 0.0962 1>
    DefaultColor=<rgba 0 1 0.5333 5>
    ForestGreen=<rgba 0 2 0.1011 1>
    Grey=<rgba 0.45 0.45 0.45 2>
    Lime=<rgba 0.7871 1 0 1>
    Orange=<rgba 1 0.4681 0 2>
    Purple=<rgba 0.6853 0.0656 1.3125 2>
    Saffron=<rgba 1 0.8214 0.2344 1>
    SkyBlue=<rgba 0 0.6654 1 1>
   )
   &VC_InitializeRibbon_v1(
    Color=$Particles.Color
    Custom Offset=true
    Enable Skeletal Position=true
    Lifetime=0.4
    Local Vector Offset=&AddVectorToPosition(Position=<0 0 0>, Vector=&AddVectorToPosition(Position=$User.DTCarOffset, Vector=<-5 0 0>))
    Offset Amount=1
    Ribbon Width=13
   )
   &AddVelocity(Velocity=<-2500 0 0>) {module=AddVelocity001}
  ]
  particleUpdateScript [
   &ParticleState()
   &InheritSourceMovement(
    Applied Position Scale=&Multiply_VectorByFloat(
     Vector=<1 1 1>
     Float=&FloatFromCurve(FloatCurve=DataInterface() {propertyValues {Curve {keys [
          {interpMode=RCIM_Cubic, leaveTangent=-0.0051, tangentMode=RCTM_Break, time=0, value=1}
          {arriveTangent=-0.0476, interpMode=RCIM_Cubic, tangentMode=RCTM_Break, time=1, value=0.9}
         ]}}})
    )
    ApplyForce=false
    ApplyPosition=true
    ApplyPositionFalloff=false
   )
   /SetVariables(
    Particles.Ribbon_BlackColor:LinearColor=<rgba 0 0 0 1>
    Particles.WaveScaleCurve:NiagaraFloat=&FloatFromCurve(
     FloatCurve=DataInterface() {propertyValues {Curve {keys [
         {interpMode=RCIM_Cubic, leaveTangentWeight=1.6473, tangentMode=RCTM_User, time=0, value=0.8}
         {
          arriveTangent=0.7219
          arriveTangentWeight=0.0795
          interpMode=RCIM_Cubic
          leaveTangent=0.1806
          leaveTangentWeight=0.0696
          tangentMode=RCTM_Break
          tangentWeightMode=RCTWM_WeightedBoth
          time=0.2
          value=1.25
         }
         <at 1 0.4 RCIM_Cubic>
        ]}}}
     Scale Curve=15
    )
   ) {isSetParameters=true}
   &ScaleColor(
    Scale Alpha=&FloatFromCurve(
     FloatCurve=DataInterface() {propertyValues {Curve {keys [
         <at 0 0 RCIM_Cubic>
         {arriveTangentWeight=0.0337, interpMode=RCIM_Cubic, leaveTangentWeight=0.0674, tangentMode=RCTM_User, time=0.2, value=1}
         {interpMode=RCIM_Cubic, leaveTangent=-6.4238, time=0.75, value=0}
        ]}}}
     Scale Curve=3
    )
    Scale RGB=&VectorFromFloat(Value=&FloatFromCurve(
      FloatCurve=DataInterface() {propertyValues {Curve {keys [<at 0 0.4>, {arriveTangent=-0.4, interpMode=RCIM_Cubic, leaveTangent=-0.4, time=0.494, value=0.506}, <at 1 0 RCIM_Cubic>]}}}
      Scale Curve=2
     ))
    ScaleA=true
    ScaleRGB=true
   )
   &SolveForcesAndVelocity(Clamp Velocity=false, Limit Acceleration=false)
   &VC_UpdateRibbon_v1(Enable Skeletal Position=true)
   /SetVariables(
    Particles.CustomPosition_R1:NiagaraPosition=&AddVectorToPosition(
     Position=$Particles.Position
     Vector=&Multiply_VectorByFloat(
      Vector=&MakeVector(
       X=&Sine(
        Normalized Angle=&Add_Float(
         A=&Multiply_Float(A=$Particles.NormalizedAge, B=-1)
         B=&Multiply_Float(A=$Emitter.LoopedAge, B=$User.WaveVelocity)
        )
        Period=$User.WavePeriod
        Scale=&Multiply_Float(
         A=&SelectFloatFromArray(Direct Array Index=$Particles.SpawnGroup, Float Selection Array=DataInterface() {propertyValues {FloatData [1,-1]}})
         B=$User.WaveMaxScale
        )
        Bias=0
       )
       Y=&Sine(
        Normalized Angle=&Add_Float(
         A=&Multiply_Float(A=$Particles.NormalizedAge, B=-1)
         B=&Multiply_Float(A=$Emitter.LoopedAge, B=$User.WaveVelocity)
        )
        Period=$User.WavePeriod
        Scale=&Multiply_Float(
         A=&SelectFloatFromArray(Direct Array Index=$Particles.SpawnGroup, Float Selection Array=DataInterface() {propertyValues {FloatData [-1,1]}})
         B=$User.WaveMaxScale
        )
        Bias=0
       )
       Z=&Cosine(
        Normalized Angle=&Add_Float(
         A=&Multiply_Float(A=$Particles.NormalizedAge, B=-1)
         B=&Multiply_Float(A=$Emitter.LoopedAge, B=$User.WaveVelocity)
        )
        Period=$User.WavePeriod
        Scale=&Multiply_Float(
         A=&SelectFloatFromArray(Direct Array Index=$Particles.SpawnGroup, Float Selection Array=DataInterface() {propertyValues {FloatData [1,-1]}})
         B=$User.WaveMaxScale
        )
        Bias=0
       )
      )
      Float=$Particles.WaveScaleCurve
     )
    )
    Particles.CustomPosition_R2:NiagaraPosition=&AddVectorToPosition(
     Position=$Particles.Position
     Vector=&Multiply_VectorByFloat(
      Vector=&MakeVector(
       X=&Sine(
        Normalized Angle=&Add_Float(
         A=&Multiply_Float(A=$Particles.NormalizedAge, B=-1)
         B=&Multiply_Float(A=$Emitter.LoopedAge, B=$User.WaveVelocity)
        )
        Period=$User.WavePeriod
        Scale=&Multiply_Float(
         A=&SelectFloatFromArray(Direct Array Index=$Particles.SpawnGroup, Float Selection Array=DataInterface() {propertyValues {FloatData [1,1]}})
         B=$User.WaveMaxScale
        )
        Bias=0
       )
       Y=&Sine(
        Normalized Angle=&Add_Float(
         A=&Multiply_Float(A=$Particles.NormalizedAge, B=-1)
         B=&Multiply_Float(A=$Emitter.LoopedAge, B=$User.WaveVelocity)
        )
        Period=$User.WavePeriod
        Scale=&Multiply_Float(
         A=&SelectFloatFromArray(Direct Array Index=$Particles.SpawnGroup, Float Selection Array=DataInterface() {propertyValues {FloatData [1,1]}})
         B=$User.WaveMaxScale
        )
        Bias=0
       )
       Z=&Cosine(
        Normalized Angle=&Add_Float(
         A=&Multiply_Float(A=$Particles.NormalizedAge, B=-1)
         B=&Multiply_Float(A=$Emitter.LoopedAge, B=$User.WaveVelocity)
        )
        Period=$User.WavePeriod
        Scale=&Multiply_Float(
         A=&SelectFloatFromArray(Direct Array Index=$Particles.SpawnGroup, Float Selection Array=DataInterface() {propertyValues {FloatData [-1,1]}})
         B=$User.WaveMaxScale
        )
        Bias=0
       )
      )
      Float=$Particles.WaveScaleCurve
     )
    )
   ) {isSetParameters=true}
   /SetVariables(Particles.CoreRibbonSize:NiagaraFloat=&Multiply_Float(
     A=$Emitter.CoreWidth
     B=&FloatFromCurve(FloatCurve=DataInterface() {propertyValues {Curve {keys [<at 0.2 1.3 RCIM_Cubic>,<at 1 0.3>]}}})
    )) {isSetParameters=true, module=SetVariables_#3}
   &ScaleRibbonWidth(Ribbon Width Scale=&FloatFromCurve(FloatCurve=DataInterface() {propertyValues {Curve {keys [<at 0 1>,<at 1 0>]}}}, Scale Curve=3))
   /SetVariables(
    Particles.BloomColor:LinearColor=&MultiplyLinearColor(
     Linear Color=$Particles.Color
     Scale Factor=&MakeVector4(X=1, Y=1, Z=1, W=&MakeCustomFloatFromBool(Bool=$User.bIsPainted, True Float=0.05, False Float=0.01))
    )
    Particles.BloomSize:NiagaraFloat=&Multiply_Float(A=$Particles.RibbonWidth, B=3)
   ) {isSetParameters=true}
  ]
  rendererClasses [&NiagaraRibbonRendererProperties]
  renderers [
   {rendererClass=&NiagaraRibbonRendererProperties}
   {rendererClass=&NiagaraRibbonRendererProperties, rendererIndex=1}
   {rendererClass=&NiagaraRibbonRendererProperties, rendererIndex=2}
   {rendererClass=&NiagaraRibbonRendererProperties, rendererIndex=3}
  ]
 }
 NE_VC_Boost_Sprite_Parent002 {
  emitterUpdateScript [
   &EmitterState(Inactive Response=::NewEnumerator1, Life Cycle Mode=::NewEnumerator1, Loop Duration=1, UseLoopDelay=false)
   &SpawnRate(SpawnRate=10, Use Spawn Probability=false) {module=SpawnRate001}
   &SpawnBurst_Instantaneous(Spawn Count=8, Use Spawn Probability=false, UseLoopCountLimit=false) {enabled=false}
   &VC_FilteredSocketCount()
   &SpawnRate(
    SpawnRate=&MakeCustomFloatFromBool(Bool=$User.bIsFrontendPreview, True Float=28, False Float=44)
    Use Spawn Probability=false
   ) {enabled=false}
  ]
  particleSpawnScript [
   &VC_PaintedVariantsOverride(
    Black=<rgba 0.2969 0.2969 0.2969 0.5>
    BurntSienna=<rgba 0.6198 0.3077 0.2453 0.6>
    Cobalt=<rgba 0.2209 0.4178 1.4941 0.7739>
    Crimson=<rgba 1.1 0.2349 0.2782 1>
    DefaultColor=&SelectLinearColorFromArray(
     Array Sampling Mode=::NewEnumerator0
     Direct Array Index=?
     Color Selection Array=DataInterface() {propertyValues {ColorData [<rgba 2 0.8349 0.4792 1>, <rgba 3 1.1615 0.6 1>]}}
     Random Seed=0
    )
    ForestGreen=<rgba 0 2 0 1>
    Grey=<rgba 0.6146 0.6146 0.6146 1>
    Lime=<rgba 0.7338 1.3 0 1>
    Orange=<rgba 2.8 0.4219 0.056 1>
    Pink=<rgba 1.3 0.5417 1.0242 0.7>
    Purple=<rgba 0.55 0 1.5 0.8>
    Saffron=<rgba 1.2 1.1391 0.4688 0.5>
    SkyBlue=<rgba 0.3764 1.1525 2 0.4>
    TitaniumWhite=<rgba 1.3047 1.3047 1.3047 0.4>
   )
   &InitializeParticle(
    Color=$Particles.Color
    Lifetime=&RandomRangeFloat(Minimum=0.4, Maximum=0.5)
    Mesh Scale=&RandomRangeVector(Minimum=<3.5 2.5 2.5>, Maximum=<5 2.5 2.5>)
    Mesh Scale Mode=::NewEnumerator4
    Sprite Rotation Mode=::NewEnumerator2
    Sprite Size Mode=::NewEnumerator3
    UsePositionOffset=false
    Write Material Random=true
   )
   &PartitionParticles(Number Of Partitions=$Emitter.PortCount)
   &SkeletalMeshLocation(
    Bone / Socket Index=$Particles.PartitionParticles.Partition
    Bone / Socket Sampling Mode=::NewEnumerator5
    MaskSpawnGroup=false
    Mesh Sampling Type=::NewEnumerator6
    Orientation Sampling=::NewEnumerator1
    Sampled Position Offset=&AddVectorToPosition(Position=<-60 0 0>, Vector=Linked() {linkedVariable {name=User.DTCarOffset, type=NiagaraPosition}})
    Skeletal Mesh=$System.SkeletalMesh
    Use Sampled Mesh Attribute as Ribbon ID=false
   )
   &ShapeLocation(
    MaskSpawnGroup=true
    Offset=<5 0 0>
    Offset Mode=::NewEnumerator0
    Override Seed=false
    Spawn Group Mask=2
    Sphere Radius=1
    UseOwnerScale=true
   ) {enabled=false}
   &AddVelocity(Velocity=<0 0 0>) {enabled=false}
   &InitialMeshOrientation(
    Mesh Axis To Orient=<0 0 -1>
    Mesh Orientation Mode=::NewEnumerator2
    Rotation=&RandomRangeVector(Minimum=<0 0 -1>, Maximum=<0 0 1>)
    Use Rotation Vector=true
   )
   &CameraOffset(Camera Offset Amount=2)
  ]
  particleUpdateScript [
   &ParticleState()
   &UpdateMeshOrientation(
    Rotation Rate=&SelectFloatFromArray(Direct Array Index=$Particles.PartitionParticles.Partition, Float Selection Array=DataInterface() {propertyValues {FloatData [3,-3]}})
    Rotation Vector=<0 0 1>
   )
   &SolveForcesAndVelocity(Clamp Velocity=false, Limit Acceleration=false)
   &ScaleColor(Scale RGB=&VectorFromFloat(Value=1), ScaleA=true, ScaleRGB=true) {module=ScaleColor001}
  ]
  rendererClasses [&NiagaraMeshRendererProperties, &NiagaraSpriteRendererProperties]
  renderers [
   {rendererClass=&NiagaraMeshRendererProperties, rendererIndex=0}
   {rendererClass=&NiagaraMeshRendererProperties, rendererIndex=1}
   {rendererClass=&NiagaraSpriteRendererProperties, rendererIndex=2}
  ]
 }
}
summary {
 systemName=NS_Boost_RageUnleashed1
 emitters [
  {
   emitterName=NE_VC_Boost_Sprite_Parent002
   rendererClasses [&NiagaraMeshRendererProperties, &NiagaraSpriteRendererProperties]
  }
  {emitterName=LightningRibbons, rendererClasses [&NiagaraRibbonRendererProperties]}
  {emitterName=CoreRibbons, rendererClasses [&NiagaraRibbonRendererProperties]}
 ]
 userVariables [
  User.bIsFrontendPreview {type=NiagaraBool}
  User.bIsPainted {type=NiagaraBool}
  User.ColorVariantID {type=NiagaraInt32}
  User.DTCarOffset {type=NiagaraPosition, description="", defaultValue {}}
  User.DTCarOffset001 {type=Vector3f, description="", defaultValue {struct=Vector3f, value=<0 0 0>}}
  User.PaintedColor {type=LinearColor, description="", defaultValue {struct=LinearColor, value=<rgba 1 1 1 1>}}
  User.PaintedScalar {type=NiagaraFloat, defaultValue=1, description=""}
  User.PortCount {type=NiagaraInt32, defaultValue=2}
  User.SingleBoostPos {type=Vector3f, description="", defaultValue {struct=Vector3f, value=<0 0 0>}}
  User.WaveMaxScale {type=NiagaraFloat, defaultValue=1, description=""}
  User.WavePeriod {type=NiagaraFloat, defaultValue=0.45, description=""}
  User.WaveVelocity {type=NiagaraFloat, defaultValue=2, description=""}
 ]
}
